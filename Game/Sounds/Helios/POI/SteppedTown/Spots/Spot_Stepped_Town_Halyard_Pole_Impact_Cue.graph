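# Looping ambient SoundCue for Stepped Town that plays intermittent halyard-on-pole impacts as a quiet outdoor spot sound.
# Each loop waits a random delay, then an even coin flip yields either silence or one of 19 equally weighted impact waves borrowed from MarinaTown, with modulated volume.
# Attenuation and concurrency are overridden on the cue, which also sends to the exterior ambient source bus.
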SoundCue /Game/Sounds/Helios/POI/SteppedTown/Spots/Spot_Stepped_Town_Halyard_Pole_Impact_Cue
root=Looping_0
nodes [
 {id=Delay_0, children [Modulator_0], properties {delayMax=4, delayMin=0.2}}
 {id=Looping_0, children [Delay_0]}
 {id=Modulator_0, children [Random_0], properties {volumeMin=0.4}}
 {id=Random_0, children [Random_1,null], properties {weights [1,1]}}
 {
  id=Random_1
  children [
   WavePlayer_21
   WavePlayer_22
   WavePlayer_23
   WavePlayer_24
   WavePlayer_25
   WavePlayer_26
   WavePlayer_27
   WavePlayer_28
   WavePlayer_29
   WavePlayer_30
   WavePlayer_31
   WavePlayer_32
   WavePlayer_33
   WavePlayer_34
   WavePlayer_35
   WavePlayer_36
   WavePlayer_37
   WavePlayer_38
   WavePlayer_39
  ]
  properties {weights [1,1,1,1,1,1,1,1,1,1,1,1,1,1,1,1,1,1,1]}
 }
 {id=WavePlayer_21, properties {soundWaveAssetPtr=/Game/Sounds/Helios/POI/MarinaTown/Spots/Spot_Chaparral_Halyard_Pole_Impact_19}}
 {id=WavePlayer_22, properties {soundWaveAssetPtr=/Game/Sounds/Helios/POI/MarinaTown/Spots/Spot_Chaparral_Halyard_Pole_Impact_18}}
 {id=WavePlayer_23, properties {soundWaveAssetPtr=/Game/Sounds/Helios/POI/MarinaTown/Spots/Spot_Chaparral_Halyard_Pole_Impact_17}}
 {id=WavePlayer_24, properties {soundWaveAssetPtr=/Game/Sounds/Helios/POI/MarinaTown/Spots/Spot_Chaparral_Halyard_Pole_Impact_16}}
 {id=WavePlayer_25, properties {soundWaveAssetPtr=/Game/Sounds/Helios/POI/MarinaTown/Spots/Spot_Chaparral_Halyard_Pole_Impact_15}}
 {id=WavePlayer_26, properties {soundWaveAssetPtr=/Game/Sounds/Helios/POI/MarinaTown/Spots/Spot_Chaparral_Halyard_Pole_Impact_14}}
 {id=WavePlayer_27, properties {soundWaveAssetPtr=/Game/Sounds/Helios/POI/MarinaTown/Spots/Spot_Chaparral_Halyard_Pole_Impact_13}}
 {id=WavePlayer_28, properties {soundWaveAssetPtr=/Game/Sounds/Helios/POI/MarinaTown/Spots/Spot_Chaparral_Halyard_Pole_Impact_12}}
 {id=WavePlayer_29, properties {soundWaveAssetPtr=/Game/Sounds/Helios/POI/MarinaTown/Spots/Spot_Chaparral_Halyard_Pole_Impact_11}}
 {id=WavePlayer_30, properties {soundWaveAssetPtr=/Game/Sounds/Helios/POI/MarinaTown/Spots/Spot_Chaparral_Halyard_Pole_Impact_10}}
 {id=WavePlayer_31, properties {soundWaveAssetPtr=/Game/Sounds/Helios/POI/MarinaTown/Spots/Spot_Chaparral_Halyard_Pole_Impact_09}}
 {id=WavePlayer_32, properties {soundWaveAssetPtr=/Game/Sounds/Helios/POI/MarinaTown/Spots/Spot_Chaparral_Halyard_Pole_Impact_08}}
 {id=WavePlayer_33, properties {soundWaveAssetPtr=/Game/Sounds/Helios/POI/MarinaTown/Spots/Spot_Chaparral_Halyard_Pole_Impact_07}}
 {id=WavePlayer_34, properties {soundWaveAssetPtr=/Game/Sounds/Helios/POI/MarinaTown/Spots/Spot_Chaparral_Halyard_Pole_Impact_06}}
 {id=WavePlayer_35, properties {soundWaveAssetPtr=/Game/Sounds/Helios/POI/MarinaTown/Spots/Spot_Chaparral_Halyard_Pole_Impact_05}}
 {id=WavePlayer_36, properties {soundWaveAssetPtr=/Game/Sounds/Helios/POI/MarinaTown/Spots/Spot_Chaparral_Halyard_Pole_Impact_04}}
 {id=WavePlayer_37, properties {soundWaveAssetPtr=/Game/Sounds/Helios/POI/MarinaTown/Spots/Spot_Chaparral_Halyard_Pole_Impact_03}}
 {id=WavePlayer_38, properties {soundWaveAssetPtr=/Game/Sounds/Helios/POI/MarinaTown/Spots/Spot_Chaparral_Halyard_Pole_Impact_02}}
 {id=WavePlayer_39, properties {soundWaveAssetPtr=/Game/Sounds/Helios/POI/MarinaTown/Spots/Spot_Chaparral_Halyard_Pole_Impact_01}}
]
properties {
 attenuationSettings=&Fort_TileAmbient_OneShot_Bird_Att
 bOverrideAttenuation=true
 bOverrideConcurrency=true
 maxDistance=5750
 soundClassObject=&Fort_Ambient_OneShots_Outdoors
 volumeMultiplier=0.251
 attenuationOverrides {
  falloffDistance=4500
  attenuationShapeExtents {x=1250}
  submixSendSettings [{
    bEnableHPFCutoff=false
    bEnableLPFCutoff=false
    disableManualSendClamp=false
    hPFCutoff=20
    lPFCutoff=20000
    maxSendDistance=6000
    maxSendLevel=1
    minSendDistance=400
    minSendLevel=0
    sendLevel=0.2
    sendLevelControlMethod=Linear
    soundSubmix=null
   }]
 }
 busSends [{
   audioBus=null
   maxSendDistance=6000
   minSendDistance=3000
   minSendLevel=0.6
   sendLevel=1
   soundSourceBus=&AmbientAudio_Exterior_SourceBus
  }]
 concurrencyOverrides {maxCount=3}
}
